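# Physics asset for the face accessory mesh of the male medium RoundThumb Gumbo character body, setting up physics for two hair bones.
# Capsule bodies cover Hair_01, Hair_02 and the head bone.
# Each hair body is constrained directly to the head with small angular limits rather than chained to the other.
# Hair_01 allows narrow swing and some twist, while Hair_02 swings evenly on both axes with twist locked.
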PhysicsAsset /BRCosmetics/Characters/Player/Male/Medium/Bodies/M_MED_RoundThumb_Gumbo/Meshes/Parts/M_MED_RoundThumb_Gumbo_FaceAcc_Physics
bodies {
 Hair_01 [{shapeType=Capsule, length=3.1735, radius=1.3031, shapeName=Hair_01_capsule, center=<0.9834 0 -1.4028>}]
 Hair_02 [{shapeType=Capsule, length=2.3659, radius=0.9917, shapeName=Hair_02_capsule, center=<1.0981 0 -0.0961>, rotation=<rot 74.9999 0.0012 0.0011>}]
 head [{shapeType=Capsule, length=3.856, radius=9.6889, center=<3.8679 -0.3093 0>, rotation=<rot 90 0 0>}]
}
constraints [
 {bone1Name=Hair_01, bone2Name=head, swing1LimitDegrees=2, swing2LimitDegrees=20, twistLimitDegrees=10}
 {bone1Name=Hair_02, bone2Name=head, swing1LimitDegrees=15, swing2LimitDegrees=15, twistLimitDegrees=0, twistMotion=Locked}
]
