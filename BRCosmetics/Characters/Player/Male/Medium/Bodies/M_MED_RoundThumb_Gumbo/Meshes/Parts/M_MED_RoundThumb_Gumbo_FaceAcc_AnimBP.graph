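# Animation Blueprint for the face accessory mesh of the Gumbo player character, adding secondary physics motion on top of the body's animation.
# It copies the pose from the attached parent mesh and runs it through a RigidBody node using the accessory's physics asset, simulated relative to neck_01 with overridden gravity.
# Two Anim Dynamics hair chains and another RigidBody node remain unconnected to the output, and the event graph has no logic.
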AnimBlueprint /BRCosmetics/Characters/Player/Male/Medium/Bodies/M_MED_RoundThumb_Gumbo/Meshes/Parts/M_MED_RoundThumb_Gumbo_FaceAcc_AnimBP
target_skeleton=&/M_MED_RoundThumb_Gumbo_FaceAcc_Skeleton
graphs [
 AnimGraph {nodes [
   n1 = Anim Dynamics - Bone: Hair_01(Alpha=1, ExternalForce="0,0,0") {
    angularDampingOverride=0.9999
    angularSpringConstant=150
    bAngularSpring=true
    bOverrideAngularDamping=true
    bUseGravityOverride=true
    simulationSpace=RootRelative
    alphaBoolBlend {blendOption=Linear}
    boundBone {boneName=Hair_01}
    componentAppliedLinearAccClamp=<1000 1000 1000>
    gravityOverride {z=-980}
    physicsBodyDefinitions [{
      boundBone {boneName=Hair_01}
      boxExtents=<7 7 7>
      constraintSetup {angularLimitsMax=<4 5 8>, angularLimitsMin=<-4 -30 -8>, angularTarget=<1 0 0>}
      localJointOffset=<0.3 0 0>
     }]
    relativeSpaceBone {boneName=head}
   }
   Anim Dynamics - Bone: Hair_02(ComponentPose=@n1, Alpha=1, ExternalForce="0,0,0") {
    angularDampingOverride=0.9999
    angularSpringConstant=200
    bAngularSpring=true
    bOverrideAngularDamping=true
    bUseGravityOverride=true
    simulationSpace=RootRelative
    alphaBoolBlend {blendOption=Linear}
    boundBone {boneName=Hair_02}
    componentAppliedLinearAccClamp=<1000 1000 1000>
    gravityOverride {z=-980}
    physicsBodyDefinitions [{
      boundBone {boneName=Hair_02}
      boxExtents=<7 7 7>
      constraintSetup {angularLimitsMax=<6 8 10>, angularLimitsMin=<-6 -10 -10>, angularTarget=<1 0 0>}
      localJointOffset=<0.3 0 0>
     }]
    relativeSpaceBone {boneName=head}
   }
   n2 = Anim Dynamics - Bone: Hair_01(Alpha=1, ExternalForce="0,0,0") {
    angularDampingOverride=0.999
    angularSpringConstant=120
    bAngularSpring=true
    bOverrideAngularDamping=true
    bUseGravityOverride=true
    simulationSpace=RootRelative
    alphaBoolBlend {blendOption=Linear}
    boundBone {boneName=Hair_01}
    componentAppliedLinearAccClamp=<10000 10000 10000>
    componentLinearAccScale=<0.2 0.4 0.2>
    componentLinearVelScale=<0.2 0.4 0.2>
    gravityOverride {z=-980}
    physicsBodyDefinitions [{
      boundBone {boneName=Hair_01}
      boxExtents=<6 6 6>
      constraintSetup {angularLimitsMax=<6 5 9>, angularLimitsMin=<-6 -35 -9>, angularTarget=<1 0 0>}
      localJointOffset=<0.5 0 0>
     }]
    relativeSpaceBone {boneName=spine_01}
   }
   Anim Dynamics - Bone: Hair_02(ComponentPose=@n2, Alpha=1, ExternalForce="0,0,0") {
    angularDampingOverride=0.999
    angularSpringConstant=200
    bAngularSpring=true
    bOverrideAngularDamping=true
    bUseGravityOverride=true
    simulationSpace=RootRelative
    alphaBoolBlend {blendOption=Linear}
    boundBone {boneName=Hair_02}
    componentAppliedLinearAccClamp=<10000 10000 10000>
    componentLinearAccScale=<0.2 0.2 0.2>
    componentLinearVelScale=<0.2 0.2 0.2>
    gravityOverride {z=-980}
    physicsBodyDefinitions [{
      boundBone {boneName=Hair_02}
      boxExtents=<5 5 5>
      constraintSetup {angularLimitsMax=<6 10 15>, angularLimitsMin=<-6 -20 -20>, angularTarget=<1 1 0>}
      localJointOffset=<0.4 0 0>
     }]
    relativeSpaceBone {boneName=spine_01}
   }
   RigidBody!(ComponentPose, Alpha, OverrideWorldGravity:Vector="0,0,0", ExternalForce) -> (Pose) {
    bClampLinearTranslationLimitToRefPose=true
    overridePhysicsAsset=&/M_MED_RoundThumb_Gumbo_FaceAcc_Physics
    simulationSpace=BaseBoneSpace
    alphaBoolBlend {blendOption=Linear}
    baseBoneRef {boneName=spine_05}
    componentAppliedLinearAccClamp=<3000 3000 3000>
    componentLinearAccScale=<0.2 0.2 0.2>
    componentLinearVelScale=<0.2 0.2 0.2>
   }
   n3 = RigidBody!(ComponentPose=Local To Component(LocalPose=Copy Pose From Mesh() {bUseAttachedParent=true}), Alpha, ExternalForce) -> (Pose) {
    bOverrideWorldGravity=true
    overridePhysicsAsset=&M_MED_RoundThumb_Gumbo_FaceAcc_Physics
    simulationSpace=BaseBoneSpace
    alphaBoolBlend {blendOption=Linear}
    baseBoneRef {boneName=neck_01}
    componentAppliedLinearAccClamp=<4000 4000 4000>
    componentLinearAccScale=<0.35 0.35 0.35>
    componentLinearVelScale=<0.35 0.35 0.35>
    overrideWorldGravity {z=-980}
   }
   Output Pose(Result=Component To Local(ComponentPose=@n3)) {layerGroup=None}
  ]}
 EventGraph {nodes [Event BlueprintUpdateAnimation()]}
]
variables []
